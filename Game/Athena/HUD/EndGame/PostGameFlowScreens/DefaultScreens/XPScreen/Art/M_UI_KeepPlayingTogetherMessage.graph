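# Translucent UI material that draws a flat-colored, angle-cut box, apparently for the Keep Playing Together message on the post-game XP screen.
# Two MF_UI_BoxAngle calls share an Angle parameter and 9-slice UVs, each cutting a different corner.
# A Min merges them into one mask, which multiplies the Color parameter's RGB and alpha.
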Material /Game/Athena/HUD/EndGame/PostGameFlowScreens/DefaultScreens/XPScreen/Art/M_UI_KeepPlayingTogetherMessage
graph {
 properties {blendMode=BLEND_Translucent, materialDomain=MD_UI, pixelDepthOffsetMode=PDOM_Legacy, usedShadingModels=Default Lit}
 nodes [
  n1 = VectorParameter() {parameterName=Color, defaultValue=<rgba 0.7913 0.0086 0.1845 1>}
  n2 = MaterialFunctionCall() {materialFunction=&GetUserInterfaceUV}
  n3 = ScalarParameter() {defaultValue=15, parameterName=Angle}
  n4 = Min(
   A=MaterialFunctionCall!(UV=@n2.9-Slice UV, Corner=Constant() {r=1}, Angle=@n3, Flip, Falloff) {materialFunction=/FortUILibrary/MaterialFunctions/Shapes/MF_UI_BoxAngle}.Result
   B=MaterialFunctionCall(UV=@n2.9-Slice UV, Corner=Constant() {r=3}, Angle=@n3) {materialFunction=&MF_UI_BoxAngle}.Result
  )
  Multiply(A=@n1.RGB, B=@n4)
  Multiply(A=@n1.A, B=@n4)
 ]
}
