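# Translucent UI material for the post-game XP screen, drawing an animated grid overlay of outlined squares over a background color.
# A particle-grid function tiles cells with eased scaling and fake-3D rotation, and a box SDF yields fill and outline masks.
# Scrolling low-res noise is sampled both pixelated and smooth.
# Reflected-gradient color blending, remapped opacity and a scalable circular transition mask shape the overlay.
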Material /Game/Athena/HUD/EndGame/PostGameFlowScreens/DefaultScreens/XPScreen/Art/M_UI_Overlay_Grid
graph {
 properties {blendMode=BLEND_Translucent, materialDomain=MD_UI, pixelDepthOffsetMode=PDOM_Legacy, usedShadingModels=Default Lit}
 nodes [
  n1 = MaterialFunctionCall() {materialFunction=&GetUserInterfaceUV}
  LinearInterpolate(
   A=VectorParameter() {parameterName=Background Color, defaultValue=<rgba 0.0024 0.016 0.117 1>}.RGB
   B=VectorParameter() {parameterName=Square Color, defaultValue=<rgba 0.0024 0.0513 0.3864 1>}.RGB
   Alpha=OneMinus(None=ComponentMask(None=@n1.Normalized UV) {g=true})
  )
  n2 = NamedRerouteUsage()
  n3 = Saturate(None=Subtract(A=@n2) {constB=0.125})
  MaterialExpressionRemap_1 = MaterialXRemap!(None=@n2, InputLow, InputHigh=OneMinus(None=@n3), TargetLow, TargetHigh)
  n4 = LinearInterpolate(A=ScalarParameter() {defaultValue=0.95, parameterName=Min Opacity, sliderMax=1}, Alpha=@MaterialExpressionRemap_1) {constA=0.95}
  Multiply(A=Multiply(A=NamedRerouteDeclaration(None=@n4) {name=Background Opacity, nodeColor=<rgba 0 0 0 1>}, B=NamedRerouteUsage())) {constB=0.8, desc=Opacity Max}
  n5 = NamedRerouteUsage()
  n6 = NamedRerouteUsage()
  n7 = Constant() {r=3}
  n8 = MaterialFunctionCall!(
   "Value (0-1)"=NamedRerouteDeclaration(None=MaterialFunctionCall(UVs=@n5, Texture Scale=@n6) {materialFunction=&ScaleUVsByCenter}.UVs) {name=Noise UV, nodeColor=<rgba 0 0 0 1>}
   Steps=Multiply(A=NamedRerouteUsage(), B=NamedRerouteDeclaration(None=@n7) {name=Noise Scale, nodeColor=<rgba 0 0 0 1>}) {constB=4}
  ) {materialFunction=/FortUILibrary/MaterialFunctions/Basics/MF_UI_StepRatio}
  n9 = Multiply(A=Time(), B=ScalarParameter() {defaultValue=0.05, parameterName=Noise Speed}) {constB=0.05}
  n10 = AppendVector(A=NamedRerouteDeclaration(None=Frac(None=@n9)) {name=Noise Rotation Speed, nodeColor=<rgba 0 0 0 1>}, B=Constant())
  n11 = TextureObject() {samplerType=SAMPLERTYPE_Grayscale, texture=/FortWater/Textures/LowResBlurredNoise}
  NamedRerouteDeclaration(None=TextureSample(UVs=Subtract(A=@n8.Stepped Gradient, B=@n10), Tex=NamedRerouteDeclaration(None=@n11) {name=Noise Texture, nodeColor=<rgba 0 0 0 1>}) {samplerSource=SSM_Wrap_WorldGroupSettings}.R) {name=Pixelated Noise, nodeColor=<rgba 0 0 0 1>}
  n12 = NamedRerouteUsage()
  n13 = AppendVector(A=NamedRerouteUsage(), B=Constant())
  NamedRerouteDeclaration(None=TextureSample(UVs=Subtract(A=@n12, B=@n13), Tex=NamedRerouteUsage()) {samplerSource=SSM_Wrap_WorldGroupSettings}.RGB) {name=Noise Gradient, nodeColor=<rgba 0 0 0 1>}
  n14 = ScalarParameter() {defaultValue=12, parameterName=Tiling}
  n15 = Constant()
  n16 = Multiply(A=MaterialFunctionCall("Value (0-1)"=NamedRerouteUsage()) {materialFunction=&MF_UI_EaseCurves}.EaseInOutQuint) {constB=2}
  n17 = NamedRerouteUsage()
  n18 = MaterialFunctionCall!(
   UV=NamedRerouteUsage()
   Global Offset
   Global Rotation
   Tiling=NamedRerouteDeclaration(None=Floor(None=@n14)) {name=Tiling Value, nodeColor=<rgba 0 0 0 1>}
   Cell Offset=@n15
   Cell Scale=Multiply(
    A=LinearInterpolate(Alpha=@n16) {constA=0.6, constB=0.7}
    B=MaterialFunctionCall("Value (0-1)"=@n17) {materialFunction=&MF_UI_EaseCurves}.EaseInQuint
   )
   Cell Rotation=@n15
   Random Seed
  ) {materialFunction=/FortUILibrary/MaterialFunctions/Basics/MF_UI_ParticleGrid}
  NamedRerouteDeclaration(None=@n18.Random Cell Values) {name=Random Cells, nodeColor=<rgba 0 0 0 1>}
  n19 = MaterialFunctionCall(UV=MaterialFunctionCall() {materialFunction=&GetUserInterfaceUV}.Normalized UV) {materialFunction=&MF_UI_UVScaling}
  NamedRerouteDeclaration(None=@n19.Scaled To Fill) {name=Scale X UV, nodeColor=<rgba 0 0 0 1>}
  n20 = VectorParameter() {parameterName=Background Color, defaultValue=<rgba 0.0024 0.016 0.117 1>}
  n21 = NamedRerouteUsage()
  n22 = LinearInterpolate(
   A=LinearInterpolate(A=@n20.RGB, B=VectorParameter() {parameterName=Square Color, defaultValue=<rgba 0.0024 0.0513 0.3864 1>}.RGB, Alpha=Multiply(A=@n21, B=@n3))
   B=VectorParameter() {parameterName=Square Outline Color, defaultValue=<rgba 0 0.0908 0.9911 1>}.RGB
   Alpha=Multiply(A=NamedRerouteUsage(), B=@n3)
  )
  n23 = MaterialFunctionCall() {materialFunction=&MF_UI_UVScaling}
  n24 = Constant() {r=0.25}
  n25 = Multiply(
   A=NamedRerouteUsage()
   B=MaterialFunctionCall("Value (0-1)"=MaterialFunctionCall!(UVs=@n23.Scaled to Fit, Offset, RotationCenter, RotationAngle=@n24) {materialFunction=/FortUILibrary/MaterialFunctions/Gradients/MF_UI_ReflectedGradient}.Result) {materialFunction=&MF_UI_EaseCurves}.EaseInOutSine
  )
  n26 = ScalarParameter() {defaultValue=1, parameterName=Color Blend, sliderMax=1}
  NamedRerouteDeclaration(None=LinearInterpolate(A=@n20.RGB, B=LinearInterpolate(A=@n20.RGB, B=@n22, Alpha=@n25), Alpha=OneMinus(None=@n26)) {constAlpha=0.25}) {name=Pattern Background Color, nodeColor=<rgba 0 0 0 1>}
  NamedRerouteDeclaration(None=@n19.Scaled to Fit) {name=Scaled UV, nodeColor=<rgba 0 0 0 1>}
  n27 = Constant()
  n28 = LinearInterpolate(Alpha=Round(None=NamedRerouteUsage())) {constA=-0.025, constB=0.025}
  n29 = Subtract(A=NamedRerouteUsage()) {constB=0.1}
  n30 = Saturate(None=Multiply(A=Saturate(None=@n29)) {constB=1.2})
  MaterialExpressionAppend3Vector_0 = MaterialXAppend3Vector!(A=@n27, B=Multiply(A=LinearInterpolate(B=@n28, Alpha=@n30) {constB=0.1}, B=NamedRerouteUsage()), C=@n27)
  n31 = MaterialFunctionCall(
   UVs=MaterialFunctionCall!(
    UV=@n18.Cells UV
    Angle=NamedRerouteDeclaration(None=@MaterialExpressionAppend3Vector_0) {name=3D UV Angle, nodeColor=<rgba 0 0 0 1>}
    Camera FOV
    Camera Offset=Constant3Vector() {constant=<rgba 0 0 2.41 0>}
    Offset from Pivot
    Texture Scale
    Texture Size
   ) {materialFunction=/FortUILibrary/MaterialFunctions/Basics/MF_UI_Rotate3D}.Rotated UV
   Size=Constant() {r=1}
  ) {materialFunction=&MF_UI_Box}
  n32 = SmoothStep(Value=OneMinus(None=AppendVector(A=@n31.SDF, B=@n31.SDF))) {constMin=0.9}
  n33 = Multiply(A=NamedRerouteDeclaration(None=ComponentMask(None=@n32) {r=true}) {name=Square Background Mask, nodeColor=<rgba 0 0 0 1>}) {constB=0.5}
  n34 = NamedRerouteUsage()
  n35 = ComponentMask(None=@n32) {g=true}
  NamedRerouteDeclaration(None=Max(A=Max(A=@n33, B=@n34), B=NamedRerouteDeclaration(None=@n35) {name=Square Outline Mask, nodeColor=<rgba 0 0 0 1>})) {name=Tile Pattern Opacity, nodeColor=<rgba 0 0 0 1>}
  n36 = MaterialFunctionCall(
   UVs=NamedRerouteUsage()
   Size=LinearInterpolate(Alpha=ScalarParameter() {defaultValue=1, parameterName=Transition Mask Scale, sliderMax=1}) {constA=-1, constB=2}
  ) {materialFunction=&MF_UI_Circle}
  NamedRerouteDeclaration(None=OneMinus(None=Saturate(None=@n36.SDF))) {name=Transition Mask, nodeColor=<rgba 0 0 0 1>}
  NamedRerouteUsage()
 ]
}
